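; BP_TakedownPrompt is a world-space text prompt actor, named for a takedown ability in the Pistol_Fang gameplay folder.
; On BeginPlay it copies Prompt_Text into its TextRender component.
; Every tick it billboards by rotating its root toward the player view point of FollowTarget's controller, changing only rotation despite that variable's name.
; The overlap event and construction script are empty.
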(event EventBeginPlay
  (Rendering|Components|TextRender|SetText (Variables|Default|GetTextRender) (Variables|Default|GetPrompt_Text)))

(event Collision|EventActorBeginOverlap
  (params (OtherActor :type Actor%20Object%20Reference)))

(event EventTick
  (params (DeltaSeconds :type Float%20%28single-precision%29))
  (Transformation|SetWorldRotation (Variables|Default|GetDefaultSceneRoot) (Math|Rotator|FindLookatRotation (Transformation|GetWorldLocation (Variables|Default|GetDefaultSceneRoot)) (Pawn|GetPlayerViewPoint (Fortnite|GetControllerfromActor (Variables|Default|GetFollowTarget)))) false false))

(fn ConstructionScript)
