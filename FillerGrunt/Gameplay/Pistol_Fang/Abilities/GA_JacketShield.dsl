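; Jacket Shield gameplay ability for the FillerGrunt Fang pistol, giving the owner a shielding pose and a damage-reduction effect while sliding.
; CanActivateAbility requires the pistol as current weapon; activation plays a montage and applies the effect.
; It ends itself on the pawn's stop-sliding event rather than tag removal, because its own movement would cancel sliding.
; Ending stops the pose montage and unbinds OnDamaged, whose handler holds only an empty server check.
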(event Custom|OnStopSliding_Event
  (params (FortPlayerPawn :type Fort%20Player%20Pawn%20Object%20Reference))
  ; We handle this ourself instead of on tag removal because we do movement in this ability that will cancel sliding state and thus cancel ability when we do not wish to do so
  (Ability|EndAbility))

(event Custom|OnDamaged
  (params (DamagedActor :type Actor%20Object%20Reference) (Damage :type Float%20%28single-precision%29) (InstigatedBy :type Controller%20Object%20Reference) (DamageCauser :type Actor%20Object%20Reference) (HitLocation :type Vector) (FHitComponent :type Primitive%20Component%20Object%20Reference) (BoneName :type Name) (Momentum :type Vector))
  (Utilities|IsServer
    (:true)))

(event Ability|EventActivateAbility
  (bind (_event_data _fort_player_pawn) (Setup|SetupGenericAbility false false)
    (:Output
      (Utilities|IsValid _fort_player_pawn
        (:Is%20Valid
          (bind _output_get (Variables|Default|SetPlayerPawn _fort_player_pawn))
          (EventDispatchers|AssignOnStopSliding _fort_player_pawn (delegate Custom|OnStopSliding_Event))
          (Ability|Tasks|PlayMontageInfoAndWaitForTrigger "None" (Variables|Animation|GetMontageInfo) :!pinned true) ; Montage
          (Ability|ApplyGameplayEffectToOwner "/FillerGrunt/Gameplay/Pistol_Fang/Abilities/GE_JacketShield_DMG_Reduc.GE_JacketShield_DMG_Reduc_C")
          (EventDispatchers|BindEventtoOnDamaged _output_get (delegate Custom|OnDamaged)))
        (:Is%20Not%20Valid
          (Ability|EndAbility))))))

(event Custom|OnTakeAnyDamage_Event
  (params (DamagedActor :type Actor%20Object%20Reference) (Damage :type Float%20%28single-precision%29) (DamageType :type Damage%20Type%20Object%20Reference) (InstigatedBy :type Controller%20Object%20Reference) (DamageCauser :type Actor%20Object%20Reference)))

(event Ability|EventOnEndAbility
  (params (bWasCancelled :type Boolean))
  ;[1 0.05306 0.6569 0.5 60] END ABILITY
  (EventDispatchers|UnbindEventfromOnDamaged (Variables|Default|GetPlayerPawn) (delegate Custom|OnDamaged))
  (Animation|StopAnimMontage (Variables|Default|GetPlayerPawn) "/FillerGrunt/Assets/Animation/Poses/M_FillerGrunt_Pistol_Pose_Shielding.M_FillerGrunt_Pistol_Pose_Shielding"))

(fn CanActivateAbility
  (params (ActorInfo :type Gameplay%20Ability%20Actor%20Info%20Structure) (Handle :type Gameplay%20Ability%20Spec%20Handle%20Structure))
  (bind _asfort_player_state (Utilities|Casting|CastToFortPlayerState (Utilities|Struct|BreakGameplayAbilityActorInfo ActorInfo))
    (:then
      (Utilities|IsValid (PlayerState|GetCurrentPawn _asfort_player_state)
        (:Is%20Valid
          (Utilities|Casting|CastToBP_FG_Fang_Pistol (Variables|/Script/FortniteGame.FortPawn|GetCurrentWeapon (PlayerState|GetCurrentPawn _asfort_player_state))
            (:then
              (return 0 true))
            (:CastFailed
              _)))
        (:Is%20Not%20Valid
          _)))
    (:CastFailed
      _))
  (return 0 false))
