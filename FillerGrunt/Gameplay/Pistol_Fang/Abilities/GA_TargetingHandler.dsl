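; Gameplay ability for the Fang pistol in FillerGrunt that detects mag-throw targets and updates the pistol's MagThrow_ValidInRange flag.
; DoDelayScan re-runs every 0.2 seconds while the player pawn is valid, prompting a scan if the Fang is equipped.
; Execute_Prompt runs a ScanForTakedown target-selection task, sets the flag and prints debug text; its target-count test (>= 0) always passes.
; Activation checks an EnableNative console variable, but both branches are empty.
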(event Ability|EventActivateAbility
  (bind _returnvalue (Development|GetConsoleVariableBoolValue "FillerGrunt.MagThrowTargeting.EnableNative"))
  (if _returnvalue
    _
    (else
      _)))

(event Custom|DoDelayScan
  (bind _playerpawn (Variables|Default|GetPlayerPawn))
  (Utilities|IsValid _playerpawn
    (:Is%20Valid
      (Utilities|Casting|CastToBP_FG_Fang_Pistol (Weapon|GetCurrentWeapon _playerpawn)
        (:then
          (CallFunction|ExecutePrompt))
        (:CastFailed
          _))
      (Utilities|FlowControl|Delay 0.2)
      (CallFunction|DoDelayScan))
    (:Is%20Not%20Valid
      _)))

(event Custom|Execute_Prompt
  (bind (_asynctaskproxy _targetdata) (Ability|Tasks|DoTargetSelection "ScanForTakedown" "(TagName=\"Ability.FillerGrunt.MagThrow.Targeting\")")
    (:Targeted
      (bind _asbp_fg_fang_pistol (Utilities|Casting|CastToBP_FG_Fang_Pistol (Weapon|GetCurrentWeapon (Variables|Default|GetPlayerPawn))))
      (if (>= (Utilities|Array|Length (Ability|TargetData|GetActorsfromTargetData _targetdata)) 0)
        (bind _asfort_player_pawn (Utilities|Casting|CastToFortPlayerPawn (Utilities|Array|Get(acopy) (Ability|TargetData|GetActorsfromTargetData _targetdata)))
          (:then
            (Variables|/FillerGrunt/Gameplay/Pistol_Fang/BP_FG_Fang_Pistol.BP_FG_Fang_Pistol_C|SetMagThrow_ValidInRange (>= (Utilities|Array|Length (Ability|TargetData|GetActorsfromTargetData _targetdata)) 0) _asbp_fg_fang_pistol)
            (Development|PrintString (Utilities|GetDisplayName _asfort_player_pawn) false false))
          (:CastFailed
            _)))
      (Variables|/FillerGrunt/Gameplay/Pistol_Fang/BP_FG_Fang_Pistol.BP_FG_Fang_Pistol_C|SetMagThrow_ValidInRange false _asbp_fg_fang_pistol)
      (Development|PrintString "NONE" false false))))

(fn CanApplyHolster
  (params (HolsterId :type Name%20%28by%20ref%29 :default "None"))
  (return false))

(fn CanActivateAbility
  (params (ActorInfo :type Gameplay%20Ability%20Actor%20Info%20Structure) (Handle :type Gameplay%20Ability%20Spec%20Handle%20Structure))
  (return 0 true))
